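# Physics asset for the Vacation mic stand instrument cosmetic, giving its hanging signs physical swing.
# Capsule bodies cover the handle and four dyn_sign bones, with sign 03 adding an end bone and thin capsules named for behind the glasses.
# Each sign is constrained to the handle with limited swing, wider on one axis, while the sign 03 end chains off sign 03 with tight limits.
# Twist is tightly limited on signs 01 and 02 and locked on the rest.
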PhysicsAsset /SparksCosmetics/Instrument/MicStand/Vacation/Meshes/Vacation_MicStand_Physics
bodies {
 dyn_sign_01 [{shapeType=Capsule, length=20.5, radius=4, shapeName=dyn_sign_01_capsule, center=<5.5969 0.1899 3.0069>, rotation=<rot 1.0046 6.3512 81.0516>}]
 dyn_sign_02 [{
   shapeType=Capsule
   length=21
   radius=5.5
   shapeName=dyn_sign_02_capsule
   center=<8.9638 -1.6769 4.8876>
   rotation=<rot 0.0003 -9.7688 85.9821>
  }]
 dyn_sign_03 [
  {shapeType=Capsule, length=1, radius=5, shapeName=dyn_sign_03_upper_string, center=<4.8586 -0.4841 4.2321>, rotation=<rot 0 -1.8 90>}
  {shapeType=Capsule, length=20, radius=0.5, shapeName=dyn_sign_03_behind_glasses_01, center=<12.7367 -0.6388 0.5526>, rotation=<rot 0 0 90>}
  {shapeType=Capsule, length=20, radius=0.5, shapeName=dyn_sign_03_behind_glasses_02, center=<17.2325 -0.6388 1.4984>, rotation=<rot 0 0 90>}
 ]
 dyn_sign_03_end [
  {shapeType=Capsule, length=35, radius=3.5, shapeName=dyn_sign_03_end_capsule_02, center=<3.2257 -0.6968 1.8454>, rotation=<rot 0 0 -90>}
  {shapeType=Capsule, length=33, radius=3.5, shapeName=dyn_sign_03_end_capsule, center=<-3.3307 -0.6968 1.7731>, rotation=<rot 0 0 -90>}
 ]
 dyn_sign_04 [{shapeType=Capsule, length=15, radius=4, shapeName=dyn_sign_04_capsule, center=<7.2427 -0.782 4.9211>, rotation=<rot 0.0001 -2.0995 90>}]
 handle [{shapeType=Capsule, length=111.0445, radius=1.679, center=<0.2319 0 10.828>}]
}
constraints [
 {bone1Name=dyn_sign_01, bone2Name=handle, swing1LimitDegrees=45, swing2LimitDegrees=3, twistLimitDegrees=5}
 {bone1Name=dyn_sign_02, bone2Name=handle, swing1LimitDegrees=45, swing2LimitDegrees=5, twistLimitDegrees=5}
 {bone1Name=dyn_sign_03, bone2Name=handle, swing1LimitDegrees=35, swing2LimitDegrees=10, twistLimitDegrees=0, twistMotion=Locked}
 {bone1Name=dyn_sign_03_end, bone2Name=dyn_sign_03, swing1LimitDegrees=5, swing2LimitDegrees=3, twistLimitDegrees=0, twistMotion=Locked}
 {bone1Name=dyn_sign_04, bone2Name=handle, swing1LimitDegrees=45, swing2LimitDegrees=10, twistLimitDegrees=0, twistMotion=Locked}
]
