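# Animation Blueprint for the Vacation mic stand instrument cosmetic that adds secondary physics motion to its skeleton.
# The input pose runs through a RigidBody node with its own physics asset, simulated in the handle bone's space with damped component motion and overridden world gravity.
# Each update sets the node's external force from skydiving and parachute state: downward normally, strongly upward in freefall, lightly downward under an open parachute.
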AnimBlueprint /SparksCosmetics/Instrument/MicStand/Vacation/Meshes/Vacation_MicStand_AnimBP
target_skeleton=&/Vacation_MicStand_Skeleton
graphs [
 AnimGraph {nodes [
   Input Pose()
   n1 = RigidBody(ComponentPose=Local To Component(LocalPose=Input Pose() {name=InPose_New}), ExternalForce=Get GravityOverride()) {
    bClampLinearTranslationLimitToRefPose=true
    bOverrideWorldGravity=true
    overridePhysicsAsset=&/Vacation_MicStand_Physics
    simulationSpace=BaseBoneSpace
    alphaBoolBlend {blendOption=Linear}
    baseBoneRef {boneName=handle}
    componentAppliedLinearAccClamp=<5000 5000 5000>
    componentLinearAccScale=<0.5 0.5 0.5>
    componentLinearVelScale=<0.1 0.1 0.1>
    overrideWorldGravity {z=-980}
   }
   Output Pose(Result=Component To Local(ComponentPose=@n1))
  ]}
 EventGraph {nodes [
   Event BlueprintUpdateAnimation()
   > n1 = Branch(Condition=Get bIsSkyDiving())
   > n2 = Branch(Condition=Get bIsParachuteOpened())
   > Set GravityOverride(GravityOverride="0,0,-100")
   Set GravityOverride(execute=@n2.else, GravityOverride="0,0,30000")
   Set GravityOverride(execute=@n1.else, GravityOverride="0,0,-980")
   TryGetPawnOwner()
  ]}
]
variables [GravityOverride {type=Vector}]
